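# Physics asset of body colliders for the medium male ArcticIceBlue player skeleton, named for use as cape collision.
# Oversized capsules cover pelvis, spine, clavicles, thighs, calves and feet, with two on the upper chest.
# Upper and lower arm bodies have no shapes, so the arms add no collision.
# Parent-to-child constraints link the bodies with swing and twist all free, leaving their angle limits inactive.
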PhysicsAsset /Game/Characters/Player/Male/Medium/Bodies/M_MED_ArcticIceBlue/Meshes/M_MED_ArcticIceBlue_CapeColliders
bodies {
 calf_l [{shapeType=Capsule, length=15, radius=25, shapeName=calf_l_capsule, center=<-15.3251 9.0072 0.0257>, rotation=<rot 90 0 0>}]
 calf_r [{shapeType=Capsule, length=15, radius=25, shapeName=calf_r_capsule, center=<15.325 -9.0072 -0.0257>, rotation=<rot 90 0 0>}]
 clavicle_l [{shapeType=Capsule, length=8, radius=13, center=<13.9836 1.7852 -1.306>, rotation=<rot 76 0 -11>}]
 clavicle_r [{shapeType=Capsule, length=8, radius=13, center=<-13.9836 -1.7852 1.306>, rotation=<rot 76 0 -11>}]
 foot_l [{shapeType=Capsule, length=14, radius=15, shapeName=foot_l_capsule, center=<-4.0998 7.0178 0.6714>, rotation=<rot 0 0 90>}]
 foot_r [{shapeType=Capsule, length=14, radius=15, shapeName=foot_r_capsule, center=<4.0998 -7.0178 -0.6714>, rotation=<rot 0 0 90>}]
 lowerarm_l []
 lowerarm_r []
 pelvis [{shapeType=Capsule, length=10, radius=13.5388, center=<-2.6215 -0.1445 0>}]
 spine_01 [{shapeType=Capsule, length=10.7187, radius=12.4138, center=<-0.759 1.721 0>}]
 spine_03 [{shapeType=Capsule, length=11.5662, radius=10.7575, center=<-1.953 2.2737 0.0001>}]
 spine_04 [{shapeType=Capsule, length=10.9259, radius=13.6325, center=<1.9087 1.5785 0.0002>}]
 spine_05 [
  {shapeType=Capsule, length=12.7812, radius=17.0625, center=<-2.7812 3.8845 0>}
  {shapeType=Capsule, length=8.7715, radius=12.0876, shapeName=spine_05_capsule, center=<7.112 3.4885 0>}
 ]
 thigh_l [{shapeType=Capsule, length=15, radius=20, shapeName=thigh_l_capsule, center=<-23.9334 7.314 1.3009>, rotation=<rot 90 0 0>}]
 thigh_r [{shapeType=Capsule, length=15, radius=20, shapeName=thigh_r_capsule, center=<23.9334 -7.314 -1.3009>, rotation=<rot 90 0 0>}]
 upperarm_l []
 upperarm_r []
}
constraints [
 {
  bone1Name=spine_01
  bone2Name=pelvis
  swing1LimitDegrees=45
  swing1Motion=Free
  swing2LimitDegrees=45
  swing2Motion=Free
  twistLimitDegrees=45
  twistMotion=Free
 }
 {
  bone1Name=spine_03
  bone2Name=spine_01
  swing1LimitDegrees=45
  swing1Motion=Free
  swing2LimitDegrees=45
  swing2Motion=Free
  twistLimitDegrees=45
  twistMotion=Free
 }
 {
  bone1Name=spine_04
  bone2Name=spine_03
  swing1LimitDegrees=45
  swing1Motion=Free
  swing2LimitDegrees=45
  swing2Motion=Free
  twistLimitDegrees=45
  twistMotion=Free
 }
 {
  bone1Name=spine_05
  bone2Name=spine_04
  swing1LimitDegrees=45
  swing1Motion=Free
  swing2LimitDegrees=45
  swing2Motion=Free
  twistLimitDegrees=45
  twistMotion=Free
 }
 {
  bone1Name=clavicle_l
  bone2Name=spine_05
  swing1LimitDegrees=45
  swing1Motion=Free
  swing2LimitDegrees=45
  swing2Motion=Free
  twistLimitDegrees=45
  twistMotion=Free
 }
 {
  bone1Name=upperarm_l
  bone2Name=clavicle_l
  swing1LimitDegrees=45
  swing1Motion=Free
  swing2LimitDegrees=45
  swing2Motion=Free
  twistLimitDegrees=45
  twistMotion=Free
 }
 {
  bone1Name=lowerarm_l
  bone2Name=upperarm_l
  swing1LimitDegrees=45
  swing1Motion=Free
  swing2LimitDegrees=45
  swing2Motion=Free
  twistLimitDegrees=45
  twistMotion=Free
 }
 {
  bone1Name=thigh_l
  bone2Name=pelvis
  swing1LimitDegrees=45
  swing1Motion=Free
  swing2LimitDegrees=45
  swing2Motion=Free
  twistLimitDegrees=45
  twistMotion=Free
 }
 {
  bone1Name=calf_l
  bone2Name=thigh_l
  swing1LimitDegrees=45
  swing1Motion=Free
  swing2LimitDegrees=45
  swing2Motion=Free
  twistLimitDegrees=45
  twistMotion=Free
 }
 {
  bone1Name=foot_l
  bone2Name=calf_l
  swing1LimitDegrees=45
  swing1Motion=Free
  swing2LimitDegrees=45
  swing2Motion=Free
  twistLimitDegrees=45
  twistMotion=Free
 }
 {
  bone1Name=clavicle_r
  bone2Name=spine_05
  swing1LimitDegrees=45
  swing1Motion=Free
  swing2LimitDegrees=45
  swing2Motion=Free
  twistLimitDegrees=45
  twistMotion=Free
 }
 {
  bone1Name=upperarm_r
  bone2Name=clavicle_r
  swing1LimitDegrees=45
  swing1Motion=Free
  swing2LimitDegrees=45
  swing2Motion=Free
  twistLimitDegrees=45
  twistMotion=Free
 }
 {
  bone1Name=lowerarm_r
  bone2Name=upperarm_r
  swing1LimitDegrees=45
  swing1Motion=Free
  swing2LimitDegrees=45
  swing2Motion=Free
  twistLimitDegrees=45
  twistMotion=Free
 }
 {
  bone1Name=thigh_r
  bone2Name=pelvis
  swing1LimitDegrees=45
  swing1Motion=Free
  swing2LimitDegrees=45
  swing2Motion=Free
  twistLimitDegrees=45
  twistMotion=Free
 }
 {
  bone1Name=calf_r
  bone2Name=thigh_r
  swing1LimitDegrees=45
  swing1Motion=Free
  swing2LimitDegrees=45
  swing2Motion=Free
  twistLimitDegrees=45
  twistMotion=Free
 }
 {
  bone1Name=foot_r
  bone2Name=calf_r
  swing1LimitDegrees=45
  swing1Motion=Free
  swing2LimitDegrees=45
  swing2Motion=Free
  twistLimitDegrees=45
  twistMotion=Free
 }
]
